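; BP_Labs_Floor_Light_A_A is a Labs set floor light Blueprint that can respond to time of day through a TOD lighting component.
; Its construction script checks the Use Time of Day flag and, when set, runs the component's RunConstruct with the Test Day Phase value to apply that phase's lighting.
; When the flag is off, it turns off auto-activation on the lighting component instead.
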(fn ConstructionScript
  (bind _todlightingcomponent (Variables|Default|GetTODLightingComponent))
  (if (Variables|Default|GetUse%20Time%20of%20Day)
    (Class|TODLightingComponent|RunConstruct _todlightingcomponent (Variables|Default|GetTest%20Day%20Phase))
    (else
      (Components|Activation|SetAutoActivate _todlightingcomponent false))))
